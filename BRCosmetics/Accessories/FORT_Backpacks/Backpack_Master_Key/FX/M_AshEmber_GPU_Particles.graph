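# Unlit translucent sprite particle material for the ash embers of the Master Key back bling, usable in Niagara and Cascade.
# An 8x8 husk-bits flipbook animates by particle age plus a random offset, mixing its channels with particle color for emissive.
# A lifetime curve-atlas ramp shapes opacity, with optional near-camera fade; particle alpha drives hue shift instead of opacity.
# Mobile uses a cheaper static single-tile sample with constant-scaled opacity.
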Material /BRCosmetics/Accessories/FORT_Backpacks/Backpack_Master_Key/FX/M_AshEmber_GPU_Particles
graph {
 properties {
  bEnableResponsiveAA=true
  bUseMaterialAttributes=true
  bUsedWithNiagaraSprites=true
  bUsedWithParticleSprites=true
  blendMode=BLEND_Translucent
  pixelDepthOffsetMode=PDOM_Legacy
  shadingModel=MSM_Unlit
  usedShadingModels=Unlit
 }
 nodes [
  n1 = TextureObject() {texture=/Game/Effects/Fort_Effects/Textures/Husk/T_Husk_Bits_SubUV}
  n2 = ParticleRelativeTime()
  n3 = ParticleRandom()
  n4 = MaterialFunctionCall!(
   Texture=@n1
   UVs
   SubImages=Constant2Vector() {g=8, r=8}
   Frame=Multiply(A=Add(A=@n2, B=@n3), B=ScalarParameter() {defaultValue=64, parameterName=Frame Speed})
  ) {materialFunction=/Engine/Functions/Engine_MaterialFunctions03/Texturing/SubUV_Function}
  n5 = MaterialFunctionCall(Float3=@n4.RGB) {materialFunction=&BreakOutFloat3Components}
  n6 = ParticleColor()
  n7 = Constant() {r=0.1}
  n8 = Multiply(A=@n6.RGB, B=Add(A=@n5.B, B=Multiply(A=@n5.G, B=@n7)))
  n9 = TextureSample(UVs=TextureCoordinate() {uTiling=0.125, vTiling=0.125}, Tex=@n1) {texture=&T_Husk_Bits_SubUV}
  n10 = Add(A=@n9.G, B=Multiply(A=@n9.B) {constB=0.2})
  n11 = CurveAtlasRowParameter(CurveTime=ParticleRelativeTime()) {
   atlas=&AthenaCurveAtlas01
   curve=/Game/Accessories/FORT_Backpacks/Backpack_Master_Key/FX/Curve_MasterKey_Ramps
   defaultValue=53
  }
  n12 = ShadingPathSwitch(Default=Multiply(A=@n11.R, B=@n4.Alpha), Mobile=Multiply(A=Constant() {r=0.7}, B=@n9.A)) {desc=Migrated from FeatureLevelSwitch}
  n13 = MaterialFunctionCall(Distance=ScalarParameter() {defaultValue=250, parameterName=Near Camera Distance}) {materialFunction=&Near_Camera_Fade}
  n14 = SetMaterialAttributes!(
   MaterialAttributes
   Emissive Color=ShadingPathSwitch(Default=Add(A=@n5.R, B=@n8), Mobile=Multiply(A=@n6.RGB, B=@n10)) {desc=Migrated from FeatureLevelSwitch}
   Opacity=StaticSwitchParameter(True=Multiply(A=@n12, B=@n13.Result), False=@n12) {parameterName=Use Near Camer Fade}
  ) {attributeSetTypes [#1, #2]}
  MaterialFunctionCall!(In=MaterialFunctionCall!(
    MA=GetMaterialAttributes(None=@n14) {attributeGetTypes [#1]}.MaterialAttributes
    InputEmissive
    HueShift=ParticleColor().A
   ) {materialFunction=/Game/Accessories/FORT_Backpacks/Backpack_Master_Key/FX/MF_ColorShift}.MA) {materialFunction=/Game/Effects/Niagara/MasterMaterials/Functions/MF_FX_EmissiveAdjustments_MA}
 ]
}
