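# Unlit translucent particle material that draws a swirling, hue-shifted center glow for the Master Key backpack FX.
# Dynamic-parameter time drives a radially twisted UV rotation into panned noise, which distorts a soft sphere mask and two optional wobbling rings.
# Emissive combines tint, noise, rings and particle color through color-shift and adjustment functions.
# Opacity uses masks and particle alpha, adding depth fade and optional Fresnel-style edge softening; mobile drops rings and both fades.
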Material /BRCosmetics/Accessories/FORT_Backpacks/Backpack_Master_Key/FX/M_CenterGlow
graph {
 properties {
  bUsedWithMeshParticles=true
  bUsedWithNiagaraMeshParticles=true
  bUsedWithParticleSprites=true
  blendMode=BLEND_Translucent
  pixelDepthOffsetMode=PDOM_Legacy
  shadingModel=MSM_Unlit
  usedShadingModels=Unlit
 }
 nodes [
  n1 = DynamicParameter() {paramNames [Time,Random,Param3,Param4]}
  n2 = Add(A=@n1.Time, B=@n1.Time)
  n3 = SphereMask(A=TextureCoordinate(), B=Constant() {r=0.5}, Radius=Constant() {r=1}, Hardness=Constant())
  n4 = Add(A=Multiply(A=@n2) {constB=0.25}, B=Multiply(A=@n3, B=@n3))
  n5 = Cosine(None=@n4)
  n6 = Subtract(A=Multiply(A=TextureCoordinate())) {constB=0.5}
  n7 = ComponentMask(None=@n6) {r=true}
  n8 = Sine(None=@n4)
  n9 = ComponentMask(None=@n6) {g=true}
  n10 = AppendVector(A=Subtract(A=Multiply(A=@n5, B=@n7), B=Multiply(A=@n8, B=@n9)), B=Add(A=Multiply(A=@n7, B=@n8), B=Multiply(A=@n9, B=@n5)))
  n11 = VectorParameter() {parameterName=UV, defaultValue=<rgba 1 1 1 0>}
  n12 = TextureSample(UVs=Multiply(A=Panner(Coordinate=@n10) {speedX=0.1, speedY=0.1}, B=ComponentMask(None=@n11.RGB) {g=true, r=true})) {samplerType=SAMPLERTYPE_LinearColor, texture=&T_NoisePack_LinearColor}
  n13 = Add(A=@n12.RGB) {constB=0.5}
  n14 = Multiply(A=@n2) {constB=0.2}
  n15 = Multiply(A=ComponentMask(None=@n12.RGB) {b=true}) {constB=0.5}
  n16 = Multiply(A=Sine(None=Add(A=@n14, B=@n15))) {constB=0.03}
  n17 = Add(A=TextureCoordinate(), B=@n16)
  n18 = Constant() {r=0.5}
  n19 = Constant() {r=0.3}
  n20 = Constant()
  n21 = TextureSample(UVs=Clamp(None=SphereMask(A=@n17, B=@n18, Radius=@n19, Hardness=@n20)) {maxDefault=0.2}) {texture=/Game/Effects/Fort_Effects/Textures/Mask/T_Rand_Pan_Mask}
  n22 = Subtract(A=TextureCoordinate(), B=@n16)
  n23 = Constant() {r=0.3}
  n24 = TextureSample(UVs=Clamp(None=SphereMask(A=@n22, B=@n18, Radius=@n23, Hardness=@n20)) {maxDefault=0.2}) {texture=&T_Rand_Pan_Mask}
  n25 = Constant()
  n26 = ShadingPathSwitch(Default=StaticSwitchParameter(True=Add(A=@n21.R, B=@n24.R), False=@n25) {defaultValue=true, parameterName=AddRings?}, Mobile=@n25) {desc=Migrated from FeatureLevelSwitch}
  n27 = TextureCoordinate()
  n28 = MaterialFunctionCall(Float3=Multiply(A=@n12.RGB) {constB=0.1}) {materialFunction=&BreakOutFloat3Components}
  n29 = SphereMask(A=Subtract(A=Add(A=@n27, B=@n28.R), B=@n28.G), B=Constant() {r=0.5}, Radius=Constant() {r=0.4}, Hardness=Constant())
  n30 = Multiply(
   A=Multiply(A=VectorParameter() {parameterName=Emissive, defaultValue=<rgba 4 1.3333 0 1>}.RGB, B=Add(A=@n13, B=@n26))
   B=Multiply(A=@n29, B=@n29)
  )
  n31 = ParticleColor()
  MaterialFunctionCall(Emissive=MaterialFunctionCall!(MA, InputEmissive=Multiply(A=@n30, B=@n31.RGB), HueShift) {materialFunction=/Game/Accessories/FORT_Backpacks/Backpack_Master_Key/FX/MF_ColorShift}.Emissive) {materialFunction=&MF_FX_EmissiveAdjustments}
  n32 = Add(A=Multiply(A=@n29) {constB=0.5}, B=Multiply(A=@n26) {constB=0.3})
  n33 = DepthFade(Opacity=@n32, FadeDistance=ScalarParameter() {defaultValue=2, parameterName=DF_FD})
  n34 = Subtract(A=DotProduct(A=CameraVectorWS(), B=VertexNormalWS())) {constB=0.2}
  n35 = LinearInterpolate(A=Saturate(None=Multiply(A=@n34) {constB=2}), B=Constant() {r=1}, Alpha=ScalarParameter() {parameterName=Not_Use_Mesh_Edge_Soft})
  Multiply(A=@n31.A, B=ShadingPathSwitch(Default=Multiply(A=@n33, B=@n35), Mobile=@n32) {desc=Migrated from FeatureLevelSwitch})
  Time()
 ]
}
