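# Masked two-sided material that dissolves the Master Key backpack over time, with a glowing edge.
# A local-space sphere mask, broken up by noise and a texture mask, is compared against a threshold animated from BP_DissolveTime.
# The result drives dithered opacity and edge emissive, while TwoSidedSign gives back faces a noisy interior glow.
# Material attributes then pass through hue-shift (BP_HueShift) and emissive adjustment functions.
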Material /BRCosmetics/Accessories/FORT_Backpacks/Backpack_Master_Key/FX/M_Dissolve
graph {
 properties {
  bUseMaterialAttributes=true
  bUsedWithMeshDeformer=true
  bUsedWithSkeletalMesh=true
  blendMode=BLEND_Masked
  pixelDepthOffsetMode=PDOM_Legacy
  twoSided=true
  usedShadingModels=Default Lit
 }
 nodes [
  n1 = TextureSample() {samplerType=SAMPLERTYPE_Normal, texture=/BRCosmetics/Accessories/FORT_Backpacks/Backpack_Master_Key/Textures/T_BackPack_Master_Key_N}
  n2 = LinearInterpolate(
   A=TextureSample() {samplerType=SAMPLERTYPE_Masks, texture=/BRCosmetics/Accessories/FORT_Backpacks/Backpack_Master_Key/Textures/T_BackPack_Master_Key_M}.R
   Alpha=DotProduct(A=Transform(None=@n1.RGB), B=CameraVectorWS())
  )
  n3 = TextureSample() {samplerType=SAMPLERTYPE_Masks, texture=/BRCosmetics/Accessories/FORT_Backpacks/Backpack_Master_Key/Textures/T_BackPack_Master_Key_S}
  n4 = TextureSample() {samplerType=SAMPLERTYPE_Normal, texture=&T_BackPack_Master_Key_N}
  n5 = SetMaterialAttributes!(
   MaterialAttributes
   Base Color=Multiply(A=TextureSample() {texture=/BRCosmetics/Accessories/FORT_Backpacks/Backpack_Master_Key/Textures/T_BackPack_Master_Key_D}.RGB, B=@n2)
   Metallic=@n3.G
   Specular=Multiply(A=@n2, B=@n3.R)
   Roughness=@n3.G
   Normal=@n4.RGB
   Ambient Occlusion=@n2
  ) {attributeSetTypes [
    #1
    #2
    #3
    #4
    #5
    #6
   ]}
  n6 = GetMaterialAttributes(None=@n5) {attributeGetTypes [#4]}
  n7 = OneMinus(None=TwoSidedSign())
  n8 = Multiply(
   A=VectorParameter() {parameterName=Emissive, defaultValue=<rgba 8 1.8863 0.3578 1>}.RGB
   B=TextureSample(UVs=TextureCoordinate() {uTiling=8, vTiling=8}) {samplerType=SAMPLERTYPE_LinearColor, texture=&T_NoisePack_LinearColor}.R
  )
  n9 = WorldPosition()
  n10 = TextureCoordinate() {uTiling=8, vTiling=8}
  n11 = Add(
   A=Add(
    A=TransformPosition(None=@n9.XYZ) {transformSourceType=TRANSFORMPOSSOURCE_World}
    B=VectorParameter() {parameterName=PositionOffset, defaultValue=<rgba 0 15 6 0>}.RGB
   )
   B=Subtract(A=TextureSample(UVs=@n10) {samplerType=SAMPLERTYPE_LinearColor, texture=&T_NoisePack_LinearColor}.RGB) {constB=0.5}
  )
  n12 = Constant()
  n13 = Constant() {r=50}
  n14 = TextureSample() {samplerType=SAMPLERTYPE_Masks, texture=&T_BackPack_Master_Key_M}
  n15 = Time()
  n16 = Add(
   A=ScalarParameter() {defaultValue=-0.5524, parameterName=BP_DissolveTime}
   B=ScalarParameter() {defaultValue=2.5, parameterName=DissolveStartOffset}
  ) {constB=1.5}
  n17 = Saturate(None=Divide(A=Subtract(A=@n15, B=@n16)) {constB=6})
  n18 = TwoSidedSign()
  n19 = Subtract(
   A=Subtract(A=SphereMask(A=@n11, B=@n12, Radius=@n13, Hardness=@n12), B=Multiply(A=@n14.G) {constB=0.25})
   B=Subtract(A=LinearInterpolate(Alpha=@n17) {constA=0.15, constB=-0.55}, B=Divide(A=@n18) {constB=20})
  )
  n20 = Add(A=@n19, B=ScalarParameter() {defaultValue=-0.9, parameterName=EdgeScale}) {constB=0.1}
  n21 = Multiply(A=Saturate(None=Multiply(A=@n20) {constB=4}), B=VectorParameter() {parameterName=EdgeEmissive, defaultValue=<rgba 3 0.7073 0.1342 1>}.RGB)
  n22 = TwoSidedSign()
  n23 = Multiply(A=@n6.Roughness, B=Add(A=Multiply(A=@n7, B=@n8), B=Multiply(A=@n21, B=@n22)))
  n24 = MaterialFunctionCall(Alpha Threshold=Multiply(A=OneMinus(None=@n19)) {constB=64}) {materialFunction=&DitherTemporalAA}
  MaterialFunctionCall!(In=MaterialFunctionCall!(
    MA=SetMaterialAttributes!(MaterialAttributes=@n6.MaterialAttributes, Opacity Mask=@n24.Result, Emissive Color=@n23) {attributeSetTypes [#7, #8]}
    InputEmissive
    HueShift=ScalarParameter() {parameterName=BP_HueShift}
   ) {materialFunction=/Game/Accessories/FORT_Backpacks/Backpack_Master_Key/FX/MF_ColorShift}.MA) {materialFunction=/Game/Effects/Niagara/MasterMaterials/Functions/MF_FX_EmissiveAdjustments_MA}
 ]
}
