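# Masked, lit mesh-particle material for the Master Key back bling effects, reusing the Master Key pickaxe textures.
# A normal-mapped, view-dependent rim term multiplies desaturated diffuse, with panning noise driving the exponent.
# The result is tinted by an Emissive color and particle alpha, then passes through hue-shift and emissive-adjustment functions.
# The same noise is added to a DissolveOffset dynamic parameter, suggesting a particle-driven dissolve.
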Material /BRCosmetics/Accessories/FORT_Backpacks/Backpack_Master_Key/FX/M_Key
graph {
 properties {bUsedWithMeshParticles=true, blendMode=BLEND_Masked, pixelDepthOffsetMode=PDOM_Legacy, usedShadingModels=Default Lit}
 nodes [
  n1 = TextureSample(UVs=Panner(Coordinate=TextureCoordinate()) {speedX=0.25, speedY=0.25}) {samplerType=SAMPLERTYPE_LinearColor, texture=&T_NoisePack_LinearColor}
  Add(A=@n1.R, B=DynamicParameter() {paramNames [Param1, Param2, DissolveOffset, HueShift]}.DissolveOffset)
  n2 = TextureSample() {samplerType=SAMPLERTYPE_Normal, texture=/BRCosmetics/Weapons/FORT_Melee/Pickaxe_MasterKey/Textures/T_Pickaxe_MasterKey_N}
  n3 = OneMinus(None=DotProduct(A=Transform(None=@n2.RGB), B=CameraVectorWS()))
  n4 = Desaturation(None=TextureSample() {texture=/BRCosmetics/Weapons/FORT_Melee/Pickaxe_MasterKey/Textures/T_Pickaxe_MasterKey_D}.RGB)
  n5 = Multiply(
   A=Power(Base=Multiply(A=@n3, B=@n4), Exp=Add(A=@n1.R))
   B=VectorParameter() {parameterName=Emissive, defaultValue=<rgba 5 1.6036 0.1823 1>}.RGB
  )
  n6 = ParticleColor()
  MaterialFunctionCall(Emissive=MaterialFunctionCall!(MA, InputEmissive=Multiply(A=@n5, B=@n6.A), HueShift) {materialFunction=/Game/Accessories/FORT_Backpacks/Backpack_Master_Key/FX/MF_ColorShift}.Emissive) {materialFunction=&MF_FX_EmissiveAdjustments}
  ScalarParameter() {defaultValue=1, parameterName=Offset}
  TextureSample() {samplerType=SAMPLERTYPE_Masks, texture=/BRCosmetics/Weapons/FORT_Melee/Pickaxe_MasterKey/Textures/T_Pickaxe_MasterKey_S}
 ]
}
