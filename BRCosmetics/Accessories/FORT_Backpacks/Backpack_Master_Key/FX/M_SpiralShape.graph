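# Unlit masked mesh-particle material for the spiral shape effect on the Master Key backpack cosmetic.
# An inverted view-facing (fresnel) term, shaped by panning noise and a contrast-boosted crack texture, scales an Emissive color through hue-shift and emissive-adjustment functions.
# Separate chains build a noise-perturbed facing value, suited to masking, and a vertex-normal offset scaled by a curve-atlas ramp along V; their output connections are not shown.
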Material /BRCosmetics/Accessories/FORT_Backpacks/Backpack_Master_Key/FX/M_SpiralShape
graph {
 properties {
  bEnableResponsiveAA=true
  bUsedWithMeshParticles=true
  blendMode=BLEND_Masked
  pixelDepthOffsetMode=PDOM_Legacy
  shadingModel=MSM_Unlit
  usedShadingModels=Unlit
 }
 nodes [
  Constant()
  n1 = DotProduct(A=VertexNormalWS(), B=CameraVectorWS())
  n2 = OneMinus(None=@n1)
  n3 = VectorParameter() {parameterName=UV, defaultValue=<rgba 0.5 4 0 0>}
  n4 = Multiply(A=TextureCoordinate(), B=AppendVector(A=@n3.R, B=@n3.R))
  n5 = TextureSample(UVs=Panner(Coordinate=@n4) {speedX=-0.2, speedY=-0.2}) {samplerType=SAMPLERTYPE_LinearColor, texture=&T_NoisePack_LinearColor}
  n6 = LinearInterpolate(Alpha=@n5.R) {constA=6, constB=2}
  n7 = Multiply(A=@n2, B=@n2)
  n8 = TextureSample(UVs=@n4) {texture=/Game/Effects/Fort_Effects/Textures/T_Cracks_Gradient}
  n9 = Add(
   A=Multiply(A=Power(Base=@n2, Exp=@n6) {constExponent=2.5}) {constB=0.01}
   B=MaterialFunctionCall(In=Multiply(A=@n7, B=@n8.R), Contrast=Constant() {r=0.02}) {materialFunction=&CheapContrast}.Result
  )
  n10 = VectorParameter() {parameterName=Emissive, defaultValue=<rgba 500 89.4365 7.3235 0>}
  MaterialFunctionCall(Emissive=MaterialFunctionCall!(MA, InputEmissive=Multiply(A=@n9, B=@n10.RGB), HueShift) {materialFunction=/Game/Accessories/FORT_Backpacks/Backpack_Master_Key/FX/MF_ColorShift}.Emissive) {materialFunction=&MF_FX_EmissiveAdjustments}
  n11 = Subtract(A=@n5.G) {constB=0.5}
  Multiply(A=Add(A=@n1, B=Multiply(A=@n11) {constB=0.3})) {constB=2}
  n12 = CurveAtlasRowParameter(CurveTime=ComponentMask(None=TextureCoordinate()) {g=true}) {
   atlas=&AthenaCurveAtlas01
   curve=/Game/Accessories/FORT_Backpacks/Backpack_Master_Key/FX/Curve_MasterKey_Ramps
   defaultValue=53
  }
  Multiply(A=VertexNormalWS(), B=@n12.G)
  OneMinus(None=@n1)
  TextureSample(UVs=@n4) {samplerType=SAMPLERTYPE_LinearColor, texture=&T_NoisePack_LinearColor}
 ]
}
