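# Niagara spark effect for the Master Key backpack cosmetic, combining glowing orange sprite sparks with mesh light rays.
# Ash spawns sparks around a sphere in looping waves that taper off, then pulls them toward an offset attractor and stretches them by speed.
# Rays repeatedly spawns randomly rotated meshes that reshape and fade through curves and dynamic material parameters.
# Ash001, a GPU spark variant without attraction, is disabled.
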NiagaraSystem /BRCosmetics/Accessories/FORT_Backpacks/Backpack_Master_Key/FX/NS_Backpack_MasterKey_Sparks
dependencies {
 usedDataInterfaces [&NiagaraDataInterfaceCurve, &NiagaraDataInterfaceVectorCurve]
 usedDynamicInputs []
 usedModules [&SystemState]
}
emitters {
 Ash {
  emitterUpdateScript [
   &EmitterState(Life Cycle Mode=::NewEnumerator1, Loop Behavior=::NewEnumerator1, Loop Duration=2, UseLoopDelay=true)
   &SpawnBurst_Instantaneous(Spawn Count=0, Use Spawn Probability=false, UseLoopCountLimit=false)
   &SpawnRate(
    SpawnRate=&Multiply_Float(
     A=100
     B=&FloatFromCurve(FloatCurve=DataInterface() {propertyValues {Curve {keys [<at 0 1>,<at 0.5 0>]}}}, CurveIndex=$Emitter.NormalizedLoopAge)
    )
    Use Spawn Probability=false
   )
  ]
  particleSpawnScript [
   &InitializeParticle(
    Lifetime=&RandomRangeFloat(Minimum=0.5, Maximum=1, Evaluation Type=::NewEnumerator1)
    Sprite Size=&RandomRangeVector2D(Minimum=<0.5 1>, Maximum=<1.5 1>, Evaluation Type=::NewEnumerator1)
    Sprite Size Mode=::NewEnumerator4
    UsePositionOffset=false
    Write Material Random=true
   )
   /CascadeToNiagaraConverter/NiagaraScripts/ModuleScripts/AddVectorToPosition(Vector:Vector3f=&TransformVector(Vector=<-5 10 0>, Destination Space=::Simulation))
   &ShapeLocation(
    Hemisphere Distribution=<0.5 1>
    MaskSpawnGroup=false
    Override Seed=false
    Rotation Mode=::NewEnumerator2
    Sphere Radius=30
    Sphere Surface Distribution=1
    UseOwnerScale=true
    Yaw / Pitch / Roll=<0.5 0 0>
   )
   /CascadeToNiagaraConverter/NiagaraScripts/ModuleScripts/FixShapeVector()
   &AddVelocity(Velocity=&Multiply_VectorByFloat(Vector=$Output.FixShapeVector.ShapeVector, Float=2))
   &Color(Color=&MakeLinearColorFromVectorAndFloat("Vector (RGB)"=<15 4.4697 0>, "Float (Alpha)"=1))
   &AddVelocity(Velocity=&RandomRangeVector(Minimum=<50 -50 0>, Maximum=<50 0 0>, Evaluation Type=::NewEnumerator1)) {module=AddVelocity001}
  ]
  particleUpdateScript [
   &ParticleState()
   &ScaleSpriteSize(
    Scale Factor=&Vector2DFromFloat(Value=&FloatFromCurve(FloatCurve=DataInterface() {propertyValues {Curve {keys [
          <at 0 0.0018 RCIM_Cubic>
          {arriveTangent=1.0147, interpMode=RCIM_Cubic, leaveTangent=1.0147, time=0.0428, value=0.1429}
          {arriveTangent=0.2202, interpMode=RCIM_Cubic, leaveTangent=0.2202, time=0.2529, value=0.2584}
          {arriveTangent=-0.151, interpMode=RCIM_Cubic, leaveTangent=-0.151, time=0.9003, value=0.3318}
          {arriveTangent=-3.3275, interpMode=RCIM_Cubic, leaveTangent=-3.3275, time=0.9666, value=0.1506}
          <at 1 0>
         ]}}}))
    Scale Sprite Size Mode=::NewEnumerator2
   )
   &Drag(Drag=2, Use Linear Drag=true, Use Rotational Drag=true)
   &PointAttractionForce(
    Attraction Radius=500
    Attraction Strength=&FloatFromCurve(
     FloatCurve=DataInterface() {propertyValues {Curve {defaultValue=1, keys [{tangentMode=RCTM_User, time=0.1, value=0}, <at 0.2 300>, <at 0.6 300>, {tangentMode=RCTM_User, time=1, value=1000}]}}}
     Scale Curve=0.5
    )
    Attractor Position Offset:Vector3f=<-10 -5 0>
    Falloff Exponent=0.2
    Kill Within Radius=false
    Module.Use Falloff=true
    UseAttractorOffset=true
   )
   &SolveForcesAndVelocity(Clamp Velocity=false, Limit Acceleration=false)
   &ScaleSpriteSizeBySpeed(Max Scale Factor=<10 50>, Source Velocity:Vector3f=&Multiply_Vector(A=<0 0.1 1>, B=$Particles.Velocity))
   &ScaleColor(
    Scale Alpha=0
    Scale RGB=&VectorFromFloat(Value=&FloatFromCurve(FloatCurve=DataInterface() {propertyValues {Curve {keys [
          <at 0.2 3170.9568 RCIM_Cubic>
          {arriveTangent=6681.8291, interpMode=RCIM_Cubic, leaveTangent=6681.8291, time=0.3, value=1515.8008}
          <at 0.75 6845.9629>
         ]}}}))
    ScaleA=true
    ScaleRGB=true
   )
  ]
  rendererClasses [&NiagaraSpriteRendererProperties]
  renderers [{rendererClass=&NiagaraSpriteRendererProperties}]
 }
 Ash001 {
  bEnabled=false
  simTarget=GPUComputeSim
  emitterUpdateScript [
   &EmitterState(Life Cycle Mode=::NewEnumerator1, Loop Behavior=::NewEnumerator1, Loop Duration=2, UseLoopDelay=true)
   &SpawnBurst_Instantaneous(Spawn Count=0, Use Spawn Probability=false, UseLoopCountLimit=false)
   &SpawnRate(
    SpawnRate=&Multiply_Float(
     A=50
     B=&FloatFromCurve(FloatCurve=DataInterface() {propertyValues {Curve {keys [<at 0 1>,<at 0.5 0>]}}}, CurveIndex=$Emitter.NormalizedLoopAge)
    )
    Use Spawn Probability=false
   )
  ]
  particleSpawnScript [
   &InitializeParticle(
    Lifetime=&RandomRangeFloat(Minimum=0.5, Maximum=2, Evaluation Type=::NewEnumerator1)
    Sprite Size=&RandomRangeVector2D(Minimum=<1 1>, Maximum=<3 1>, Evaluation Type=::NewEnumerator1)
    Sprite Size Mode=::NewEnumerator4
    UsePositionOffset=false
    Write Material Random=true
   )
   &AddVectorToPosition(Vector:Vector3f=&TransformVector(Vector=<-5 10 0>, Destination Space=::Simulation))
   &ShapeLocation(
    Hemisphere Distribution=<0.5 1>
    MaskSpawnGroup=false
    Override Seed=false
    Rotation Mode=::NewEnumerator2
    Sphere Radius=30
    Sphere Surface Distribution=1
    UseOwnerScale=true
    Yaw / Pitch / Roll=<0.5 0 0>
   )
   &FixShapeVector()
   &AddVelocity(Velocity=&Multiply_VectorByFloat(
     Vector=$Output.FixShapeVector.ShapeVector
     Float=&RandomRangeFloat(Minimum=2, Maximum=4, Evaluation Type=::NewEnumerator1)
    ))
   &Color(Color=&MakeLinearColorFromVectorAndFloat("Vector (RGB)"=<15 4.4697 0>, "Float (Alpha)"=1))
   &AddVelocity(Velocity=&RandomRangeVector(Minimum=<-30 30 -30>, Maximum=<30 30 30>, Evaluation Type=::NewEnumerator1)) {module=AddVelocity001}
  ]
  particleUpdateScript [
   &ParticleState()
   &ScaleSpriteSize(
    Scale Factor=&Vector2DFromFloat(Value=&FloatFromCurve(FloatCurve=DataInterface() {propertyValues {Curve {keys [
          <at 0 0.0018 RCIM_Cubic>
          {arriveTangent=1.8817, interpMode=RCIM_Cubic, leaveTangent=1.8817, time=0.0428, value=0.7389}
          {arriveTangent=-0.6169, interpMode=RCIM_Cubic, leaveTangent=-0.6169, time=0.2529, value=0.4776}
          {arriveTangent=-0.6393, interpMode=RCIM_Cubic, leaveTangent=-0.6393, time=0.8631, value=0.2328}
          <at 1 0>
         ]}}}))
    Scale Sprite Size Mode=::NewEnumerator2
   )
   &Drag(Drag=3, Use Linear Drag=true, Use Rotational Drag=true)
   &SolveForcesAndVelocity(Clamp Velocity=false, Limit Acceleration=false)
   &ScaleColor(
    Scale Alpha=0
    Scale RGB=&VectorFromFloat(Value=&FloatFromCurve(FloatCurve=DataInterface() {propertyValues {Curve {keys [
          <at 0.0037 5968.9941 RCIM_Cubic>
          {arriveTangent=95.308, interpMode=RCIM_Cubic, leaveTangent=95.308, time=0.3, value=709.966}
          <at 0.75 6040.1265>
         ]}}}))
    ScaleA=true
    ScaleRGB=true
   )
  ]
  rendererClasses [&NiagaraSpriteRendererProperties]
  renderers [{rendererClass=&NiagaraSpriteRendererProperties}]
 }
 Rays {
  emitterUpdateScript [
   &EmitterState(Life Cycle Mode=::NewEnumerator1, Loop Behavior=::NewEnumerator1, Loop Delay=0.2, Loop Duration=0.5, UseLoopDelay=true)
   &SpawnBurst_Instantaneous(Spawn Count=5, Use Spawn Probability=false, UseLoopCountLimit=false)
   &SpawnRate(SpawnRate=6, Use Spawn Probability=false)
  ]
  particleSpawnScript [
   /SetVariables(Particles.MeshRotationEuler:Vector3f=&RandomRangeVector(Minimum=<-0.15 -0.15 -0.35>, Maximum=<0.15 0.15 -0.15>, Evaluation Type=::NewEnumerator1)) {isSetParameters=true}
   /SetVariables(Particles.MeshOrientation=/CascadeToNiagaraConverter/NiagaraScripts/DynamicInputScripts/MakeQuatFromEuler(Euler:Vector3f=$Particles.MeshRotationEuler, Scale:NiagaraFloat=360)) {isSetParameters=true}
   &InitializeParticle(
    Lifetime=&RandomRangeFloat(Minimum=0.25, Maximum=1, Evaluation Type=::NewEnumerator1)
    Mesh Scale=&VectorFromFloat(Value=&RandomRangeFloat(Minimum=0.3, Maximum=0.2, Evaluation Type=::NewEnumerator1))
    Mesh Scale Mode=::NewEnumerator4
    UsePositionOffset=false
    Write Material Random=true
   )
   &AddVectorToPosition(Vector:Vector3f=&TransformVector(Vector=<-5 -2 0>, Destination Space=::Simulation))
   &DynamicMaterialParameters(
    Index 0 Param 2=&RandomRangeFloat(Minimum=0, Maximum=10, Evaluation Type=::NewEnumerator1)
    Module.Param0WriteEnabled=false
    Module.Param1WriteEnabled=true
    Module.Param2WriteEnabled=false
    Module.Param3WriteEnabled=false
   ) {module=DynamicMaterialParameters001}
  ]
  particleUpdateScript [
   &ParticleState()
   &ScaleMeshSize(Scale Factor=&VectorFromCurve(VectorCurve=DataInterface() {propertyValues {
       XCurve {keys [
         {arriveTangent=1.4598, interpMode=RCIM_Cubic, leaveTangent=1.4598, tangentMode=RCTM_User, time=0, value=0.0635}
         {arriveTangent=0.0313, interpMode=RCIM_Cubic, leaveTangent=0.0313, tangentMode=RCTM_User, time=1, value=0.75}
        ]}
       YCurve {keys [
         {arriveTangent=-9.6811, interpMode=RCIM_Cubic, leaveTangent=-9.6811, tangentMode=RCTM_User, time=0, value=5}
         {interpMode=RCIM_Cubic, tangentMode=RCTM_User, time=1, value=0.1046}
        ]}
       ZCurve {keys [
         {arriveTangent=-9.4039, interpMode=RCIM_Cubic, leaveTangent=-9.4039, tangentMode=RCTM_User, time=0, value=5}
         {interpMode=RCIM_Cubic, tangentMode=RCTM_User, time=1, value=0.1046}
        ]}
      }}))
   &Color(Color=&MakeLinearColorFromVectorAndFloat(
     "Vector (RGB)"=&VectorFromFloat(Value=&FloatFromCurve(FloatCurve=DataInterface() {propertyValues {Curve {keys [<at 0.1142 0 RCIM_Cubic>, {arriveTangent=-2.4725, leaveTangent=-2.4725, time=0.5076, value=2.4725}, <at 1 0>]}}}))
     "Float (Alpha)"=&ClampFloat(Float=&FloatFromCurve(FloatCurve=DataInterface() {propertyValues {Curve {keys [
           <at 0.004 0 RCIM_Cubic>
           {arriveTangent=-0.6898, interpMode=RCIM_Cubic, leaveTangent=-0.6898, tangentMode=RCTM_User, time=0.1276, value=0.2296}
           <at 1 0 RCIM_Cubic>
          ]}}}))
    ))
   &DynamicMaterialParameters(
    Index 0 Param 1=&FloatFromCurve(FloatCurve=DataInterface() {propertyValues {Curve {keys [{arriveTangent=2.0018, time=0, value=0}, <at 1 2.0018>]}}})
    Index 0 Param 3=1
    Module.Param0WriteEnabled=true
    Module.Param1WriteEnabled=false
    Module.Param2WriteEnabled=true
    Module.Param3WriteEnabled=true
   )
   &CameraOffset(Camera Offset Amount=1)
  ]
  rendererClasses [&NiagaraMeshRendererProperties]
  renderers [{rendererClass=&NiagaraMeshRendererProperties, rendererIndex=0}]
 }
}
summary {
 systemName=NS_Backpack_MasterKey_Sparks
 emitters [
  {emitterName=Ash, rendererClasses [&NiagaraSpriteRendererProperties]}
  {bEnabled=false, emitterName=Ash001, simTarget=GPUComputeSim, rendererClasses [&NiagaraSpriteRendererProperties]}
  {emitterName=Rays, rendererClasses [&NiagaraMeshRendererProperties]}
 ]
 userVariables []
}
