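# Anim Blueprint for the Master Key backpack, adjusting its position on the player and tracking a door-open state.
# Spine yaw relative to spine_01 drives alphas for additive backpack_root tilts over the input pose.
# An Athena kill-stat watcher, or a ten-second clock check when custom animations are allowed, sets DoorOpen, cleared after 60 updates.
# The door-bone blend is not wired to the output pose.
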AnimBlueprint /BRCosmetics/Accessories/FORT_Backpacks/Backpack_Master_Key/Meshes/M_MED_Master_Key_Backpack_AnimBlueprint
target_skeleton=&/M_MED_Master_Key_Backpack_Skeleton
graphs [
 AnimGraph {nodes [
   n1 = "Transform (Modify) Bone - Bone: dyn_l_door"(Alpha=1, Translation="0,0,0", Rotation="0.000,90,0.000", Scale="1,1,1") {
    rotationMode=BMM_Additive
    rotationSpace=BCS_BoneSpace
    alphaBoolBlend {blendOption=Linear}
    boneToModify {boneName=dyn_l_door}
    rotation {yaw=90}
   }
   n2 = "Transform (Modify) Bone - Bone: dyn_r_door"(ComponentPose=@n1, Alpha=1, Translation="0,0,0", Rotation="0,90,0", Scale="1,1,1") {
    rotationMode=BMM_Additive
    rotationSpace=BCS_BoneSpace
    alphaBoolBlend {blendOption=Linear}
    boneToModify {boneName=dyn_r_door}
    rotation {yaw=90}
   }
   Blend Poses by bool(BlendPose_0=Component To Local(ComponentPose=@n2)) {blendType=Linear}
   AnimGraphNode_SubInput_0 = Input Pose()
   n3 = "Transform (Modify) Bone - Bone: backpack_root"(ComponentPose=Local To Component(LocalPose=@AnimGraphNode_SubInput_0), Alpha=1, Translation="0,1.2,0", Rotation="0,0,0", Scale="1,1,1") {
    translationMode=BMM_Additive
    translationSpace=BCS_BoneSpace
    alphaBoolBlend {blendOption=Linear}
    boneToModify {boneName=backpack_root}
    translation {y=1.2}
   }
   n4 = "Transform (Modify) Bone - Bone: backpack_root"(ComponentPose=@n3, Alpha=Get backpack_rotate_dn() -> (backpack_rotate_dn:double), Translation="1,0,0", Rotation="0,20,0", Scale="1,1,1") {
    rotationMode=BMM_Additive
    rotationSpace=BCS_BoneSpace
    translationMode=BMM_Additive
    translationSpace=BCS_BoneSpace
    alphaBoolBlend {blendOption=Linear}
    boneToModify {boneName=backpack_root}
    rotation {yaw=20}
    translation {x=1}
   }
   n5 = "Transform (Modify) Bone - Bone: backpack_root"(ComponentPose=@n4, Alpha=Get backpack_rotate_up() -> (backpack_rotate_up:double), Translation="-0.5,0,0", Rotation="0,-30,0", Scale="1,1,1") {
    rotationMode=BMM_Additive
    rotationSpace=BCS_BoneSpace
    translationMode=BMM_Additive
    translationSpace=BCS_BoneSpace
    alphaBoolBlend {blendOption=Linear}
    boneToModify {boneName=backpack_root}
    rotation {yaw=-30}
    translation {x=-0.5}
   }
   Output Pose(Result=Component To Local(ComponentPose=@n5)) {layerGroup=None}
   Get DoorOpen() -> (DoorOpen:bool)
  ]}
 EventGraph {nodes [
   Event BlueprintUpdateAnimation()
   n1 = GetOwnerSkeletalMeshComponent()
   n2 = BreakRotator(InRot=BreakTransform(InTransform=GetDeltaTransformFromRefPose(self=@n1, BoneName=spine_05, BaseName=spine_01)).Rotation)
   > Set backpack_rotate_dn(backpack_rotate_dn:double=MapRangeClamped(Value=@n2.Yaw, InRangeA=0, InRangeB=-30)) -> (Output_Get:double)
   > Set backpack_rotate_up(backpack_rotate_up:double=MapRangeClamped(Value=@n2.Yaw, InRangeA=0, InRangeB=30)) -> (Output_Get:double)
   n3 = Sequence!(execute) -> (then_0, then_1, then_2:Exec, then_3:Exec)
   > Do Once()
   > n4 = Cast To FortPlayerPawn(Object=TryGetFortPawn())
   n5 = WatchCosmeticStat!(
    execute:Exec=@n3.then_1
    Pawn:Fort Player Pawn Object Reference=@n4
    StatTag:Fort Stat Manager Tag Structure=(TagName="StatManager.AthenaKills")
    DefaultValue:int=0
    bBroadcastInitialValue:bool=true
    InitialValueBroadcastDelay:float=-1
    ChangeValueBroadcastDelay:float=-1
    bConsumeWaitingNotifications:bool=true
   ) -> (then:Exec, OnStatChanged:Exec, StatName:Name, StatValue:int) {is=title K2Node_AsyncAction}
   Branch(
    execute=@n5.OnStatChanged
    Condition="float > float"(A="To Float (Integer)"(InInt=@n5.StatValue), B=Get KillCount() -> (KillCount:double)) {is=title K2Node_CallFunction}
   )
   > Set Active_Kill(Active_Kill:bool=true) -> (Output_Get:bool)
   > Set KillCount(KillCount:double=float + float(A=Get KillCount(), B=1) {is=title K2Node_CommutativeAssociativeBinaryOperator}) -> (Output_Get:double)
   n6 = Branch(execute=@n3.then_2, Condition=Get Active_Kill() -> (Active_Kill:bool))
   > Set DoorOpen(DoorOpen:bool=true) -> (Output_Get:bool)
   > Set Active_Kill(Active_Kill=false)
   > Set OpenTimer(OpenTimer:double=0) -> (Output_Get:double)
   Branch(execute=@n6.else, Condition=Get bCanPlayCustomAnimations() -> (bCanPlayCustomAnimations:bool))
   n7 = Get Second!(A:Date Time Structure=Now!() -> (ReturnValue:Date Time Structure)) -> (ReturnValue:int=0) {is=title K2Node_CallFunction}
   > Branch(Condition="Equal (Integer)"(A="% (Integer)"(A=@n7, B=10) -> (ReturnValue=0), B=0) {is=title K2Node_CallFunction})
   > Set DoorOpen(DoorOpen=true)
   > Set OpenTimer(OpenTimer=0)
   Branch(execute=@n3.then_3, Condition=Get DoorOpen())
   > n8 = Set OpenTimer(OpenTimer=float + float(A=Get OpenTimer() -> (OpenTimer:double), B=1))
   > Branch(Condition="Equal (Float)"!(A:double=@n8, B:double=60) -> (ReturnValue:bool) {is=title K2Node_CallFunction})
   > Set DoorOpen(DoorOpen=false)
   > Set OpenTimer(OpenTimer=0)
  ]}
]
variables [
 Active_Kill {type=boolean}
 backpack_rotate_dn {type=number}
 backpack_rotate_up {type=number}
 DoorOpen {type=boolean}
 KillCount {type=number}
 OpenTimer {type=number}
]
